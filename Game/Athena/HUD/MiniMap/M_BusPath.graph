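# Translucent UI material for the bus path on the Athena minimap, drawing a blue-tinted arrow stripe.
# It derives horizontal UVs from the widget's pixel size and a rounding step, intended to tile only whole arrows of the GoFasterStripe texture.
# A panner along the same coordinate drives a remapped sine that modulates the texture's alpha, sending a brightness wave along the path.
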Material /Game/Athena/HUD/MiniMap/M_BusPath
graph {
 properties {blendMode=BLEND_Translucent, materialDomain=MD_UI, pixelDepthOffsetMode=PDOM_Legacy, usedShadingModels=Default Lit}
 nodes [
  n1 = MaterialFunctionCall() {materialFunction=&GetUserInterfaceUV}
  n2 = MaterialFunctionCall(Float2=@n1.Pixel Size) {materialFunction=&BreakOutFloat2Components}
  n3 = Round(None=Divide(A=@n2.R, B=@n2.R))
  n4 = MaterialFunctionCall(Float2=@n1.Normalized UV) {materialFunction=&BreakOutFloat2Components}
  n5 = Divide(
   A=Multiply(A=Multiply(A=@n3, B=@n2.G), B=@n4.R)
   B=MaterialFunctionCall(Float2=@n1.Pixel Size) {materialFunction=&BreakOutFloat2Components}.G
  ) {constB=20, desc=The UV tialed so that we always have a whole arrow}
  n6 = TextureSample(UVs=MaterialFunctionCall(X=@n5, Y=@n4.G) {materialFunction=&MakeFloat2}.Result) {texture=&/GoFasterStripe}
  Multiply(A=@n6.RGB, B=Constant3Vector() {constant=<rgba 0 0.646 1 0>})
  n7 = Panner(Coordinate=Multiply(A=@n5) {constB=0.2}) {constCoordinate=1, speedX=-1}
  n8 = Constant3Vector() {constant=<rgba 0.25 0.25 0.25 0>}
  Multiply(
   A=@n6.A
   B=Saturate(None=MaterialFunctionCall!(Sine Phase=@n7, Value 1=@n8, Value 2) {materialFunction=/Engine/Functions/Engine_MaterialFunctions02/Utility/Sine_Remapped}.Remapped Sine)
  )
 ]
}
