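# Translucent UI material for the middle line of the bus path on the minimap.
# UVs come from pixel-size math meant to tile whole arrows, though as wired the scaling cancels to roughly plain UV.
# IsDash, on by default, selects the static MidLineStripe texture for color and alpha.
# When off, color uses a panning barrier stripe, and alpha oscillates on a time sine between it and a static second stripe.
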Material /Game/Athena/HUD/MiniMap/M_BusPathMidLine
graph {
 properties {blendMode=BLEND_Translucent, materialDomain=MD_UI, pixelDepthOffsetMode=PDOM_Legacy, usedShadingModels=Default Lit}
 nodes [
  n1 = ScalarParameter() {defaultValue=1, parameterName=IsDash, sliderMax=1}
  n2 = MaterialFunctionCall() {materialFunction=&GetUserInterfaceUV}
  n3 = MaterialFunctionCall(Float2=@n2.Pixel Size) {materialFunction=&BreakOutFloat2Components}
  n4 = Multiply(A=Round(None=Divide(A=@n3.R, B=@n3.R)), B=@n3.G)
  n5 = MaterialFunctionCall(Float2=@n2.Normalized UV) {materialFunction=&BreakOutFloat2Components}
  n6 = MaterialFunctionCall(
   X=Divide(
    A=Multiply(A=@n4, B=@n5.R)
    B=MaterialFunctionCall(Float2=@n2.Pixel Size) {materialFunction=&BreakOutFloat2Components}.G
   ) {constB=20, desc=The UV tialed so that we always have a whole arrow}
   Y=@n5.G
  ) {materialFunction=&MakeFloat2}
  n7 = TextureSample(UVs=Panner(Coordinate=@n6.Result) {speedX=0.04}) {texture=&/T-UI-BarrierStripe}
  n8 = Time()
  n9 = TextureSample(UVs=@n6.Result) {texture=&/MidLineStripe}
  LinearInterpolate(
   A=LinearInterpolate(A=TextureSample(UVs=@n6.Result) {texture=&/T-UI-BarrierStripe2}.A, B=@n7.A, Alpha=Sine(None=@n8) {period=2.7})
   B=@n9.A
   Alpha=@n1
  )
  Multiply(A=LinearInterpolate(A=@n7.RGB, B=@n9.RGB, Alpha=@n1), B=Constant3Vector() {constant=<rgba 1 1 1 0>})
 ]
}
