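# Minimap material function that tests whether the current UV lies inside a circular safe zone.
# It measures UV distance to a center built from the zone data's R on both axes (G is unused), against a radius of B plus an optional additive.
# Threshold outputs 1 inside and 0 outside, contrary to its own output description.
# SignedDistanceToEdge outputs distance minus radius, negative inside.
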MaterialFunction /Game/Athena/HUD/MiniMap/MF_SafeZoneTest
graph {
 properties {description="Returns 1 if inside the specified safe zone, and 0 otherwise"}
 nodes [
  n1 = MaterialFunctionCall(Float3=FunctionInput() {inputName=InSafeZoneData, previewValue=<0.5 0.5 0.2 1>}) {materialFunction=&BreakOutFloat3Components}
  n2 = Distance(
   A=TextureCoordinate()
   B=MaterialFunctionCall(X=@n1.R, Y=@n1.R) {materialFunction=&MakeFloat2}.Result
  )
  n3 = Add(A=@n1.B, B=FunctionInput(Preview=Constant()) {bUsePreviewValueAsDefault=true, inputName=InRadiusAdditive, inputType=FunctionInput_Scalar})
  FunctionOutput(None=If(A=@n2, B=@n3, "A > B"=Constant(), "A < B"=Constant() {r=1})) {description="0: Inside circle, 1: outside circle", outputName=Threshold}
  FunctionOutput(None=Subtract(A=@n2, B=@n3)) {outputName=SignedDistanceToEdge}
 ]
}
