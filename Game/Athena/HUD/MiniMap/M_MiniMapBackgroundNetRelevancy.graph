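# Translucent UI material for the Athena HUD minimap background that tints the area outside a net relevancy zone.
# The CircleData parameter's RGB feeds MF_SafeZoneTest, reusing the safe-zone circle test to produce a Threshold mask.
# That mask blends the TintOutsideRelevancyZone color and alpha toward zero, by default a semi-transparent black.
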Material /Game/Athena/HUD/MiniMap/M_MiniMapBackgroundNetRelevancy
graph {
 properties {blendMode=BLEND_Translucent, materialDomain=MD_UI, pixelDepthOffsetMode=PDOM_Legacy, usedShadingModels=Default Lit}
 nodes [
  n1 = VectorParameter() {group=EffectParams, parameterName=TintOutsideRelevancyZone, defaultValue=<rgba 0 0 0 0.6>}
  n2 = MaterialFunctionCall!(InRadiusAdditive, InSafeZoneData=VectorParameter() {group=Inputs, parameterName=CircleData, defaultValue=<rgba 0.3073 0.7 0.145 1>}.RGB) {materialFunction=&/MF_SafeZoneTest}
  LinearInterpolate(A=@n1.RGB, Alpha=@n2.Threshold) {constB=0}
  LinearInterpolate(A=@n1.A, Alpha=@n2.Threshold) {constB=0}
 ]
}
