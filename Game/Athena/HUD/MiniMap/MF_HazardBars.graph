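# Material function in the minimap HUD folder that outputs a 0/1 mask of repeating, screen-aligned hazard bars.
# Screen UVs are panned by BarSpeed (static by default), then the doubled X component is scaled by BarMultiplier, giving vertical bars.
# A sine of that value is compared with Threshold for hard edges, higher thresholds giving thinner bars.
# The safe-zone and circle descriptions do not match the graph, which has no circle test.
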MaterialFunction /Game/Athena/HUD/MiniMap/MF_HazardBars
graph {
 properties {description="Returns 1 if inside the specified safe zone, and 0 otherwise"}
 nodes [
  n1 = Constant()
  n2 = MaterialFunctionCall(Float2=Panner(
    Coordinate=MaterialFunctionCall() {materialFunction=&ScreenAlignedUVs}."X 100%, Y 100%"
    Speed=FunctionInput(Preview=@n1) {bUsePreviewValueAsDefault=true, inputName=BarSpeed, inputType=FunctionInput_Scalar}
   )) {materialFunction=&BreakOutFloat2Components}
  n3 = Multiply(
   A=Add(A=@n2.R, B=@n2.R)
   B=FunctionInput(Preview=Constant() {r=5}) {bUsePreviewValueAsDefault=true, inputName=BarMultiplier, inputType=FunctionInput_Scalar}
  ) {constB=5}
  n4 = FunctionInput(Preview=Constant() {r=0.9}) {
   bUsePreviewValueAsDefault=true
   description="-1 .. 1 threshold on a sine wave, higher values are thinner"
   inputName=Threshold
   inputType=FunctionInput_Scalar
  }
  FunctionOutput(None=If(A=Sine(None=@n3), B=@n4, "A > B"=Constant() {r=1}, "A < B"=Constant())) {description="0: Inside circle, 1: outside circle", outputName=Threshold}
  TextureCoordinate()
 ]
}
