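# Material function for the minimap that draws an antialiased circle outline from center, radius, thickness and falloff inputs.
# It measures UV distance from the center and scales thickness and falloff by PixelToUVDensity, converting pixel sizes to UV units.
# Two SmoothStep calls form soft inner and outer edges around the radius, and their difference outputs the ring mask.
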MaterialFunction /Game/Athena/HUD/MiniMap/MF_MinimapCircleMaker2
graph {nodes [
  n1 = FunctionInput() {description=Center of the circle, inputName=Center, inputType=FunctionInput_Vector2, previewValue=<0.5 0.5 0 1>}
  n2 = Distance(A=TextureCoordinate(), B=@n1)
  n3 = FunctionInput() {description=Radius of the circle, inputName=Radius, inputType=FunctionInput_Scalar, sortPriority=1, previewValue=<0.6 0 0 1>}
  n4 = FunctionInput() {bUsePreviewValueAsDefault=true, inputName=Thickness, inputType=FunctionInput_Scalar, sortPriority=2, previewValue=<1 1 1 1>}
  n5 = FunctionInput() {bUsePreviewValueAsDefault=true, inputName=PixelToUVDensity, inputType=FunctionInput_Scalar, sortPriority=3, previewValue=<0.001 1 1 1>}
  n6 = Multiply(A=Divide(A=@n4), B=@n5)
  n7 = Subtract(A=@n3, B=@n6)
  n8 = FunctionInput() {bUsePreviewValueAsDefault=true, inputName=Falloff, inputType=FunctionInput_Scalar, sortPriority=2, previewValue=<1 1 1 1>}
  n9 = Multiply(A=@n8, B=@n5)
  n10 = MaterialFunctionCall(Alpha=@n2, Min=Subtract(A=@n7, B=@n9), Max=@n7) {materialFunction=&SmoothStep}
  n11 = Add(A=@n3, B=@n6)
  n12 = MaterialFunctionCall(Alpha=@n2, Min=@n11, Max=Add(A=@n11, B=@n9)) {materialFunction=&SmoothStep}
  FunctionOutput(None=Subtract(A=Add(A=@n10.Result), B=Add(A=@n12.Result))) {description=Draws a nice AA'd circle, outputName=Circle}
 ]}
