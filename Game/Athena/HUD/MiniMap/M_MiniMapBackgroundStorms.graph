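# Translucent UI material that draws the storm area on the minimap, tinting outside the safe circle read from MPC_Minimap's CurrentCircle.
# Safe-zone tests and hazard bars blend tint with danger-line color, and an outside-only circle falloff adds shadow color.
# A screen-aligned grid texture, animated by a time-driven sine, modulates opacity within that falloff.
# A static switch optionally rescales UVs around the center by the collection's ScaleFactor.
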Material /Game/Athena/HUD/MiniMap/M_MiniMapBackgroundStorms
graph {
 properties {blendMode=BLEND_Translucent, materialDomain=MD_UI, numCustomizedUVs=1, pixelDepthOffsetMode=PDOM_Legacy, usedShadingModels=Default Lit}
 nodes [
  n1 = VectorParameter() {group=EffectParams, parameterName=DangerLineColor, defaultValue=<rgba 0.2008 0.0063 0.3906 0.5>}
  n2 = VectorParameter() {group=EffectParams, parameterName=TintOutsideSafeZone, defaultValue=<rgba 0.3515 0.011 0.6795 0.5>}
  n3 = ComponentMask(None=CollectionParameter() {collection=&/MPC_Minimap, parameterName=CurrentCircle}) {b=true, g=true, r=true}
  n4 = MaterialFunctionCall!(InRadiusAdditive, InSafeZoneData=@n3) {materialFunction=&/MF_SafeZoneTest}
  n5 = LinearInterpolate(A=@n2.RGB, Alpha=@n4.Threshold) {constB=0}
  n6 = MaterialFunctionCall!(
   Threshold=ScalarParameter() {defaultValue=0.3, parameterName=PatternWidth}
   BarSpeed=ScalarParameter() {parameterName="BarSpeed (S)"}
   BarMultiplier=ScalarParameter() {defaultValue=120, parameterName=PatternRepeat}
  ) {materialFunction=&/MF_HazardBars}
  n7 = ScalarParameter() {parameterName=HazardGap}
  n8 = Clamp(None=Add(
    A=OneMinus(None=@n6.Threshold)
    B=MaterialFunctionCall(InRadiusAdditive=@n7, InSafeZoneData=@n3) {materialFunction=&MF_SafeZoneTest}.Threshold
   )) {minDefault=0.25}
  n9 = CollectionParameter() {collection=&MPC_Minimap, parameterName=CurrentCircle}
  n10 = Multiply(
   A=OneMinus(None=@n4.Threshold)
   B=MaterialFunctionCall!(
    Center=ComponentMask(None=@n9) {g=true, r=true}
    Radius=ComponentMask(None=@n9) {b=true}
    Thickness=ScalarParameter() {parameterName=Thickness_OuterRing}
    Falloff=ScalarParameter() {defaultValue=116.3512, parameterName=CurrentCircleFalloff}
    PixelToUVDensity
   ) {materialFunction=&/MF_MinimapCircleMaker2}.Circle
  )
  Reroute(None=LinearInterpolate(
    A=LinearInterpolate(A=@n1.RGB, B=@n5, Alpha=@n8)
    B=VectorParameter() {parameterName=HazardShadow, defaultValue=<rgba 0.7448 0 0.5735 0.498>}.RGB
    Alpha=@n10
   ))
  n11 = LinearInterpolate(A=@n2.A, Alpha=@n4.Threshold) {constB=0}
  n12 = ScalarParameter() {defaultValue=1, parameterName=FeatureSwitch}
  n13 = ScalarParameter() {defaultValue=0.2, parameterName=Ripple to Grid Scale}
  n14 = TextureSample(UVs=Multiply(
    A=MaterialFunctionCall!(TextureResolution) {materialFunction=/Engine/Functions/Engine_MaterialFunctions02/Texturing/ScreenAlignedPixelToPixelUVs}.Result
    B=ScalarParameter() {defaultValue=0.25, parameterName=PatternScale}
   )) {samplerType=SAMPLERTYPE_LinearColor, texture=/Game/Gadgets/Assets/Teleporter/Textures/T_Teleporter_GridSquares_01}
  n15 = ComponentMask(None=@n14.RGB) {r=true}
  n16 = Multiply(A=Time() {bOverride_Period=true, period=3.8314}, B=ScalarParameter() {defaultValue=0.261, parameterName=Sine Speed Grid})
  n17 = Sine(None=Add(A=Multiply(A=@n12, B=@n13), B=Add(A=@n15, B=@n16)))
  n18 = ComponentMask(None=@n14.RGB) {g=true}
  n19 = Multiply(A=@n10, B=Power(Base=Multiply(A=@n17, B=@n18), Exp=ScalarParameter() {defaultValue=5, parameterName=Grid Power})) {constA=1}
  Saturate(None=Add(A=LinearInterpolate(A=@n1.A, B=@n11, Alpha=@n8) {constA=1}, B=ConstantBiasScale(None=@n19) {scale=0.1}))
  n20 = TextureCoordinate()
  n21 = CollectionParameter() {collection=&MPC_Minimap, parameterName=ScaleFactor}
  n22 = OneMinus(None=@n21)
  StaticSwitchParameter(True=Add(A=Multiply(A=@n20, B=@n21) {constB=2}, B=Multiply(A=@n22) {constB=0.5}), False=@n20) {defaultValue=true, parameterName=UseFullScreenScale}
 ]
}
