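# M_MotherShip_Shadow is a translucent UI material in the minimap HUD folder, named as the shadow for the mothership.
# A RadialGradientExponential function, driven by Radius and Density parameters, produces a circular mask with a high default density for a tight falloff.
# The mask is multiplied by an Opacity scalar parameter, alongside a zero constant that suggests a flat black color.
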Material /Game/Athena/HUD/MiniMap/M_MotherShip_Shadow
graph {
 properties {blendMode=BLEND_Translucent, materialDomain=MD_UI, pixelDepthOffsetMode=PDOM_Legacy, usedShadingModels=Default Lit}
 nodes [
  Constant()
  Multiply(
   A=MaterialFunctionCall(Radius=ScalarParameter() {defaultValue=0.5, parameterName=Radius}, Density=ScalarParameter() {defaultValue=48, parameterName=Density}) {materialFunction=&RadialGradientExponential}.RadialGradientExponential
   B=ScalarParameter() {defaultValue=0.33, parameterName=Opacity}
  )
 ]
}
