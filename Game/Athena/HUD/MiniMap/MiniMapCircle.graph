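# Translucent UI material that draws the current and next circles on the Athena minimap.
# Centers and radii come from MPC_Minimap and feed MF_MinimapCircleMaker2, with pixel-to-UV density derived from the UI pixel size.
# The current circle pulses between two tints on a time-driven sine, while the next circle layers a ring over an outline, with summed, saturated alpha.
# The UseFullScreenScale static switch adjusts UVs and line density by the collection's ScaleFactor.
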Material /Game/Athena/HUD/MiniMap/MiniMapCircle
graph {
 properties {blendMode=BLEND_Translucent, materialDomain=MD_UI, numCustomizedUVs=1, pixelDepthOffsetMode=PDOM_Legacy, usedShadingModels=Default Lit}
 nodes [
  n1 = VectorParameter() {group=EffectParams, parameterName=CurrentCircleTint1, defaultValue=<rgba 0.5938 0 0.5596 1>}
  n2 = VectorParameter() {group=EffectParams, parameterName=CurrentCircleTint, defaultValue=<rgba 0.2232 0.007 0.4342 0.4>}
  n3 = MaterialFunctionCall(Value=Time()) {materialFunction=&LinearSine}
  n4 = CollectionParameter() {collection=&/MPC_Minimap, parameterName=CurrentCircle}
  n5 = MaterialFunctionCall() {materialFunction=&GetUserInterfaceUV}
  n6 = Divide(A=@n5.Pixel Size, B=Max(B=CollectionParameter() {collection=&MPC_Minimap, parameterName=ScaleFactor}) {constA=0.1})
  n7 = Multiply(A=Divide(B=StaticSwitchParameter(True=@n6, False=@n5.Pixel Size) {parameterName=UseFullScreenScale}) {constA=900, constB=1024}) {constB=0.001}
  n8 = MaterialFunctionCall!(
   Center=ComponentMask(None=@n4) {g=true, r=true}
   Radius=ComponentMask(None=@n4) {b=true}
   Thickness=ScalarParameter() {defaultValue=0.4, parameterName=Thickness_OuterRing}
   Falloff=ScalarParameter() {defaultValue=2, parameterName=CurrentCircleFalloff}
   PixelToUVDensity=@n7
  ) {materialFunction=&/MF_MinimapCircleMaker2}
  n9 = CollectionParameter() {collection=&MPC_Minimap, parameterName=NextCircle}
  n10 = ComponentMask(None=@n9) {g=true, r=true}
  n11 = ComponentMask(None=@n9) {b=true}
  n12 = MaterialFunctionCall(
   Center=@n10
   Radius=@n11
   Thickness=ScalarParameter() {defaultValue=1, parameterName=Thickness_Outline}
   Falloff=ScalarParameter() {defaultValue=6, parameterName=Falloff_Outline}
   PixelToUVDensity=@n7
  ) {materialFunction=&MF_MinimapCircleMaker2}
  n13 = VectorParameter() {group=EffectParams, parameterName=NextCircleOutlintTint, defaultValue=<rgba 0 0 0 0.2>}
  n14 = MaterialFunctionCall(
   Center=@n10
   Radius=@n11
   Thickness=ScalarParameter() {defaultValue=0.5, parameterName=Thickness_Ring}
   Falloff=ScalarParameter() {defaultValue=2, parameterName=Falloff_Ring}
   PixelToUVDensity=@n7
  ) {materialFunction=&MF_MinimapCircleMaker2}
  n15 = VectorParameter() {group=EffectParams, parameterName=NextCircleTint, defaultValue=<rgba 1 1 1 1>}
  Add(
   A=Multiply(A=LinearInterpolate(A=@n1.RGB, B=@n2.RGB, Alpha=@n3.Linear Sine), B=@n8.Circle)
   B=LinearInterpolate(A=Multiply(A=@n12.Circle, B=@n13.RGB), B=Multiply(A=@n14.Circle, B=@n15.RGB), Alpha=@n14.Circle)
  )
  n16 = Multiply(A=@n8.Circle, B=LinearInterpolate(A=@n1.A, B=@n2.A, Alpha=@n3.Linear Sine))
  n17 = Multiply(A=@n13.A, B=@n12.Circle)
  Saturate(None=Add(A=Add(A=@n16, B=@n17), B=Multiply(A=@n14.Circle, B=@n15.A)))
  n18 = TextureCoordinate()
  n19 = CollectionParameter() {collection=&MPC_Minimap, parameterName=ScaleFactor}
  n20 = OneMinus(None=@n19)
  StaticSwitchParameter(True=Add(A=Multiply(A=@n18, B=@n19) {constB=2}, B=Multiply(A=@n20) {constB=0.5}), False=@n18) {parameterName=UseFullScreenScale}
  VectorParameter() {group=Inputs, parameterName=NextCircle, defaultValue=<rgba 0.5 0.5 0.3 1>}
  VectorParameter() {group=Inputs, parameterName=CurrentCircle, defaultValue=<rgba 0.5 0.5 0.6 1>}
 ]
}
