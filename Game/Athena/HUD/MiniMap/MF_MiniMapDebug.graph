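# MF_MiniMapDebug is a material function that draws a colored debug circle for the minimap.
# It unpacks one four-component input into center (RG), radius (B) and color blend (A), passing center, radius and a ScaleFactor parameter to MF_MinimapCircleMaker.
# The circle mask is output as opacity, and also multiplies a lerp between the FirstCircleColor and LastCircleColor parameters to give the final color.
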MaterialFunction /Game/Athena/HUD/MiniMap/MF_MiniMapDebug
graph {nodes [
  n1 = FunctionInput() {inputName=DebugCircleData, inputType=FunctionInput_Vector4}
  n2 = MaterialFunctionCall!(
   Center=ComponentMask(None=@n1) {g=true, r=true}
   Radius=ComponentMask(None=@n1) {b=true}
   Scale Factor=ScalarParameter() {defaultValue=2.75, parameterName=ScaleFactor}
  ) {materialFunction=&/MF_MinimapCircleMaker}
  n3 = VectorParameter() {group=EffectParams, parameterName=FirstCircleColor, defaultValue=<rgba 1 1 1 1>}
  n4 = VectorParameter() {group=EffectParams, parameterName=LastCircleColor, defaultValue=<rgba 1 0 0 1>}
  n5 = ComponentMask(None=@n1) {a=true}
  FunctionOutput(None=Multiply(A=@n2.Circle, B=LinearInterpolate(A=@n3.RGB, B=@n4.RGB, Alpha=@n5))) {outputName=FinalColor}
  FunctionOutput(None=@n2.Circle) {outputName=Opacity}
 ]}
