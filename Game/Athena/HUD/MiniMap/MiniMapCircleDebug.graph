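# Translucent UI material for the minimap that draws up to fourteen debug circles.
# Each circle's data comes from its own DebugCircle parameter in the MPC_Minimap parameter collection and feeds a separate call to the MF_MiniMapDebug function.
# Two chains of Add nodes sum the per-circle FinalColor and Opacity results, so overlapping circles combine additively without clamping in this graph.
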Material /Game/Athena/HUD/MiniMap/MiniMapCircleDebug
graph {
 properties {blendMode=BLEND_Translucent, materialDomain=MD_UI, pixelDepthOffsetMode=PDOM_Legacy, usedShadingModels=Default Lit}
 nodes [
  n1 = MaterialFunctionCall!(DebugCircleData=CollectionParameter() {collection=&/MPC_Minimap, parameterName=DebugCircle_0}) {materialFunction=&/MF_MiniMapDebug}
  n2 = MaterialFunctionCall(DebugCircleData=CollectionParameter() {collection=&MPC_Minimap, parameterName=DebugCircle_1}) {materialFunction=&MF_MiniMapDebug}
  n3 = Add(A=@n1.Opacity, B=@n2.Opacity)
  n4 = MaterialFunctionCall(DebugCircleData=CollectionParameter() {collection=&MPC_Minimap, parameterName=DebugCircle_2}) {materialFunction=&MF_MiniMapDebug}
  n5 = MaterialFunctionCall(DebugCircleData=CollectionParameter() {collection=&MPC_Minimap, parameterName=DebugCircle_3}) {materialFunction=&MF_MiniMapDebug}
  n6 = MaterialFunctionCall(DebugCircleData=CollectionParameter() {collection=&MPC_Minimap, parameterName=DebugCircle_4}) {materialFunction=&MF_MiniMapDebug}
  n7 = Add(A=Add(A=Add(A=@n3, B=@n4.Opacity), B=@n5.Opacity), B=@n6.Opacity)
  n8 = MaterialFunctionCall(DebugCircleData=CollectionParameter() {collection=&MPC_Minimap, parameterName=DebugCircle_5}) {materialFunction=&MF_MiniMapDebug}
  n9 = MaterialFunctionCall(DebugCircleData=CollectionParameter() {collection=&MPC_Minimap, parameterName=DebugCircle_6}) {materialFunction=&MF_MiniMapDebug}
  n10 = MaterialFunctionCall(DebugCircleData=CollectionParameter() {collection=&MPC_Minimap, parameterName=DebugCircle_7}) {materialFunction=&MF_MiniMapDebug}
  n11 = Add(A=Add(A=Add(A=@n7, B=@n8.Opacity), B=@n9.Opacity), B=@n10.Opacity)
  n12 = MaterialFunctionCall(DebugCircleData=CollectionParameter() {collection=&MPC_Minimap, parameterName=DebugCircle_8}) {materialFunction=&MF_MiniMapDebug}
  n13 = MaterialFunctionCall(DebugCircleData=CollectionParameter() {collection=&MPC_Minimap, parameterName=DebugCircle_9}) {materialFunction=&MF_MiniMapDebug}
  n14 = MaterialFunctionCall(DebugCircleData=CollectionParameter() {collection=&MPC_Minimap, parameterName=DebugCircle_10}) {materialFunction=&MF_MiniMapDebug}
  n15 = Add(A=Add(A=Add(A=@n11, B=@n12.Opacity), B=@n13.Opacity), B=@n14.Opacity)
  n16 = MaterialFunctionCall(DebugCircleData=CollectionParameter() {collection=&MPC_Minimap, parameterName=DebugCircle_11}) {materialFunction=&MF_MiniMapDebug}
  n17 = MaterialFunctionCall(DebugCircleData=CollectionParameter() {collection=&MPC_Minimap, parameterName=DebugCircle_12}) {materialFunction=&MF_MiniMapDebug}
  n18 = MaterialFunctionCall(DebugCircleData=CollectionParameter() {collection=&MPC_Minimap, parameterName=DebugCircle_13}) {materialFunction=&MF_MiniMapDebug}
  Add(A=Add(A=Add(A=@n15, B=@n16.Opacity), B=@n17.Opacity), B=@n18.Opacity)
  n19 = Add(A=@n1.FinalColor, B=@n2.FinalColor)
  n20 = Add(A=Add(A=Add(A=@n19, B=@n4.FinalColor), B=@n5.FinalColor), B=@n6.FinalColor)
  n21 = Add(A=Add(A=Add(A=@n20, B=@n8.FinalColor), B=@n9.FinalColor), B=@n10.FinalColor)
  n22 = Add(A=Add(A=Add(A=@n21, B=@n12.FinalColor), B=@n13.FinalColor), B=@n14.FinalColor)
  Add(A=Add(A=Add(A=@n22, B=@n16.FinalColor), B=@n17.FinalColor), B=@n18.FinalColor)
 ]
}
