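# MiniMapNetRelevancyCircle is a translucent UI material that, going by its name, draws the net relevancy circle on the Athena minimap.
# The CircleData parameter packs center in RG and radius in B, which feed MF_MinimapCircleMaker along with a ScaleFactor parameter.
# The function's Circle output is multiplied by the RGB of the CircleTint parameter.
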Material /Game/Athena/HUD/MiniMap/MiniMapNetRelevancyCircle
graph {
 properties {blendMode=BLEND_Translucent, materialDomain=MD_UI, pixelDepthOffsetMode=PDOM_Legacy, usedShadingModels=Default Lit}
 nodes [
  n1 = VectorParameter() {group=Inputs, parameterName=CircleData, defaultValue=<rgba 0.5 0.5 0.6 1>}
  Multiply(
   A=VectorParameter() {group=EffectParams, parameterName=CircleTint, defaultValue=<rgba 0 0 0 0.6>}.RGB
   B=MaterialFunctionCall!(
    Center=ComponentMask(None=@n1.RGB) {g=true, r=true}
    Radius=@n1.B
    Scale Factor=ScalarParameter() {defaultValue=1.5, parameterName=ScaleFactor}
   ) {materialFunction=&/MF_MinimapCircleMaker}.Circle
  )
 ]
}
