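# Translucent UI material that draws the upcoming circle on the minimap, reading the NextNextCircle parameter from the minimap parameter collection.
# That parameter's RG channels give center and B gives radius to MF_MinimapCircleMaker, along with a scale factor.
# The resulting circle mask is multiplied by a tint for color.
# It is also multiplied by an offset, scaled, time-driven sine, producing a pulsing value.
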Material /Game/Athena/HUD/MiniMap/MiniMapNextCircle
graph {
 properties {blendMode=BLEND_Translucent, materialDomain=MD_UI, pixelDepthOffsetMode=PDOM_Legacy, usedShadingModels=Default Lit}
 nodes [
  n1 = CollectionParameter() {collection=&/MPC_Minimap, parameterName=NextNextCircle}
  n2 = MaterialFunctionCall!(
   Center=ComponentMask(None=@n1) {g=true, r=true}
   Radius=ComponentMask(None=@n1) {b=true}
   Scale Factor=ScalarParameter() {defaultValue=2.75, parameterName=ScaleFactor}
  ) {materialFunction=&/MF_MinimapCircleMaker}
  Multiply(A=@n2.Circle, B=VectorParameter() {group=EffectParams, parameterName=NextCircleTint, defaultValue=<rgba 1 1 1 1>}.RGB)
  n3 = ScalarParameter() {defaultValue=1.5, parameterName=SineDuration}
  n4 = Sine(None=Multiply(A=Time(), B=ScalarParameter() {defaultValue=0.75, parameterName=TimeDuration}))
  Multiply(A=@n2.Circle, B=Multiply(A=Add(A=@n3, B=@n4), B=ScalarParameter() {defaultValue=0.35, parameterName=SineClamp}))
 ]
}
